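; BP_PilgrimShaderWarmupInterface is a Blueprint interface that, by its name, serves shader warmup in PilgrimCore.
; It declares two functions, GetMaterialsAndMeshes and GetNiagaraSystems, whose names indicate queries for the materials, meshes and Niagara systems to warm up.
; Both functions are stubs with no logic, leaving the content to implementing classes.
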(fn GetMaterialsAndMeshes
  (return 0))

(fn GetNiagaraSystems
  (return 0))
